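# Masked, two-sided, default-lit master material for Niagara sprite and mesh particles, filed under debris effects and built around SubUV flipbook textures.
# SubUV texture parameters supply base color, opacity mask and normal, with base color multiplied by particle color.
# A static switch takes opacity from either the base color alpha or a selectable channel of the separate mask texture.
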Material /BRCosmetics/Effects/Niagara/MasterMaterials/Debris/M_SubUV_Masked_Base
graph {
 properties {
  bUsedWithNiagaraMeshParticles=true
  bUsedWithNiagaraSprites=true
  blendMode=BLEND_Masked
  pixelDepthOffsetMode=PDOM_Legacy
  twoSided=true
  usedShadingModels=Default Lit
 }
 nodes [
  n1 = TextureSampleParameterSubUV!(UVs, Apply View MipBias) {parameterName=Base Color}
  Multiply(A=ParticleColor().RGB, B=@n1.RGB)
  n2 = TextureSampleParameterSubUV() {
   parameterName=Opacity Mask
   samplerType=SAMPLERTYPE_Masks
   texture=/Game/Characters/Player/Female/Medium/Bodies/F_MED_Commando_01/Textures/SolidWhite_MaskCompression
  }
  StaticSwitchParameter(True=@n1.A, False=ChannelMaskParameter(None=@n2.RGB) {parameterName=Opacity Mask Channel}.R) {parameterName=Opacity from Base Color Alpha}
  TextureSampleParameterSubUV() {parameterName=Normal, samplerType=SAMPLERTYPE_Normal, texture=/Engine/EngineMaterials/BaseFlattenNormalMap}
 ]
}
